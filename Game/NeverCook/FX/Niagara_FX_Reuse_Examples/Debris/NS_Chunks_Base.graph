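# Niagara system serving as a base debris chunk effect, with one GPU sprite emitter that continuously spawns small, short-lived, brown-tinted particles.
# Particles spawn within a sphere volume with varied color, size and lifetime, get an initial velocity, then fall under gravity.
# A float curve drives sprite scale, shrinking chunks to zero over the latter half of the curve.
# ForwardVector and LandingTier user parameters are exposed, but no shown module input reads them.
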NiagaraSystem /Game/NeverCook/FX/Niagara_FX_Reuse_Examples/Debris/NS_Chunks_Base
dependencies {usedDataInterfaces [&NiagaraDataInterfaceCurve], usedDynamicInputs [], usedModules [&SystemState]}
emitters {Chunks001 {
  simTarget=GPUComputeSim
  emitterUpdateScript [
   &EmitterState()
   &SpawnRate(SpawnRate=100, Use Spawn Probability=false)
  ]
  particleSpawnScript [
   &InitializeParticle(
    Color Maximum=<rgba 0.2083 0.139 0.0696 1>
    Color Minimum=<rgba 0.333 0.2498 0.1665 1>
    Color Mode=::NewEnumerator2
    Lifetime Max=1
    Lifetime Min=0.5
    Lifetime Mode=::NewEnumerator1
    Mass Mode=::NewEnumerator1
    Sprite Rotation Mode=::NewEnumerator2
    Sprite Size Mode=::NewEnumerator1
    Sprite UV Mode=::NewEnumerator3
    Uniform Sprite Size Max=8
    Uniform Sprite Size Min=4
    UsePositionOffset=false
    Write Material Random=true
   )
   &SphereLocation(
    MaskSpawnGroup=false
    Override Seed=false
    Sphere Radius=40
    Surface Only=false
    UseNonUniformScale=false
    UseOffset=false
    UseSphereOrientation=false
   )
   &AddVelocity(Module.Use Falloff Distance=false, Override Seed=false, Velocity Mode=::NewEnumerator1, Velocity Speed=200)
  ]
  particleUpdateScript [
   &ParticleState()
   &GravityForce()
   &SolveForcesAndVelocity(Clamp Velocity=false, Limit Acceleration=false)
   &ScaleSpriteSize(Scale Factor=&Vector2DFromFloat(Value=&FloatFromCurve(FloatCurve=DataInterface() {propertyValues {Curve {keys [<at 0.5 1 RCIM_Cubic>, {arriveTangent=-3.7562, interpMode=RCIM_Cubic, leaveTangent=-3.7562, tangentMode=RCTM_User, time=1, value=0}]}}})))
  ]
  rendererClasses [&NiagaraSpriteRendererProperties]
  renderers [{rendererClass=&NiagaraSpriteRendererProperties}]
 }}
summary {
 systemName=NS_Chunks_Base
 emitters [{emitterName=Chunks001, simTarget=GPUComputeSim, rendererClasses [&NiagaraSpriteRendererProperties]}]
 userVariables [
  User.ForwardVector {type=Vector3f, description="", defaultValue {struct=Vector3f, value=<0 0 0>}}
  User.LandingTier {type=NiagaraInt32}
 ]
}
